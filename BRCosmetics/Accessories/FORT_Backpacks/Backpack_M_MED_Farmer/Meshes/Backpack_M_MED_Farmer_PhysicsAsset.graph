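# Physics asset for the Farmer backpack cosmetic, giving ragdoll-style bodies to a straw figure with head, arms and hands.
# A box on backpack_root anchors the straw-body rotator, both arms and the head; the rotator and arms rotate freely, while the head has tight swing and locked twist.
# Two straw-body capsules chain from the rotator, and two hand capsules chain from each arm, with narrow swing limits and near-zero or locked twist.
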PhysicsAsset /BRCosmetics/Accessories/FORT_Backpacks/Backpack_M_MED_Farmer/Meshes/Backpack_M_MED_Farmer_PhysicsAsset
bodies {
 backpack_root [{shapeType=Box, extentX=21.5027, extentY=4.8964, extentZ=17.0777, center=<-6.1156 0.7256 0>}]
 dyn_Strawbody_0 [{shapeType=Capsule, length=4.25, radius=5.75, center=<6.039 0 -0.1681>, rotation=<rot -88.4961 180 -179.9999>}]
 dyn_Strawbody_1 [{shapeType=Capsule, length=14.8824, radius=4.1187, center=<10.244 -0.1128 -0.7677>, rotation=<rot 2.0685 -90 -89.9313>}]
 dyn_Strawbody_rotater [{shapeType=Box, extentX=28.5, extentY=10, extentZ=33, center=<15.6923 11.048 -0.0004>, rotation=<rot 0 -4.0214 0>}]
 dyn_alm_l_01 [{shapeType=Capsule, length=12.75, radius=2.5, center=<-0.2171 0.0002 -7.2777>, rotation=<rot 11.4941 0 0>}]
 dyn_arm_r_01 [{shapeType=Capsule, length=12.75, radius=2.5, center=<0.007 0 6.057>, rotation=<rot -13.0492 0 0>}]
 dyn_hand_l_01 [{shapeType=Capsule, length=1.75, radius=2.5, center=<0.6904 0 -0.1511>, rotation=<rot 77.6559 0 0>}]
 dyn_hand_l_02 [{shapeType=Capsule, length=3.75, radius=2, center=<1.7948 -0.0002 -0.3928>, rotation=<rot 77.6538 0 0>}]
 dyn_hand_r_01 [{shapeType=Capsule, length=1.25, radius=3, center=<0.6184 0 0.2469>, rotation=<rot 68.2371 180 180>}]
 dyn_hand_r_02 [{shapeType=Capsule, length=3.75, radius=2, center=<0.7675 0.0001 0.5879>, rotation=<rot 68.2357 180 180>}]
 dyn_head [{shapeType=Capsule, length=6.1147, radius=9.3903, center=<11.6925 0.4965 -0.3751>, rotation=<rot 82.8664 73.1825 -111.1743>}]
}
constraints [
 {bone1Name=dyn_Strawbody_1, bone2Name=dyn_Strawbody_0, swing1LimitDegrees=5, swing2LimitDegrees=2, twistLimitDegrees=0.01}
 {bone1Name=dyn_Strawbody_0, bone2Name=dyn_Strawbody_rotater, swing1LimitDegrees=25, swing2LimitDegrees=10, twistLimitDegrees=0.01}
 {bone1Name=dyn_hand_r_01, bone2Name=dyn_arm_r_01, swing1LimitDegrees=10, swing2LimitDegrees=20, twistLimitDegrees=1, twistMotion=Locked}
 {bone1Name=dyn_hand_r_02, bone2Name=dyn_hand_r_01, swing1LimitDegrees=5, swing2LimitDegrees=10, twistLimitDegrees=1, twistMotion=Locked}
 {bone1Name=dyn_hand_l_01, bone2Name=dyn_alm_l_01, swing1LimitDegrees=10, swing2LimitDegrees=20, twistLimitDegrees=1, twistMotion=Locked}
 {bone1Name=dyn_hand_l_02, bone2Name=dyn_hand_l_01, swing1LimitDegrees=5, swing2LimitDegrees=10, twistLimitDegrees=1, twistMotion=Locked}
 {
  bone1Name=dyn_Strawbody_rotater
  bone2Name=backpack_root
  swing1LimitDegrees=45
  swing1Motion=Free
  swing2LimitDegrees=45
  swing2Motion=Free
  twistLimitDegrees=45
  twistMotion=Free
 }
 {
  bone1Name=dyn_arm_r_01
  bone2Name=backpack_root
  swing1LimitDegrees=45
  swing1Motion=Free
  swing2LimitDegrees=45
  swing2Motion=Free
  twistLimitDegrees=45
  twistMotion=Free
 }
 {
  bone1Name=dyn_alm_l_01
  bone2Name=backpack_root
  swing1LimitDegrees=45
  swing1Motion=Free
  swing2LimitDegrees=45
  swing2Motion=Free
  twistLimitDegrees=45
  twistMotion=Free
 }
 {bone1Name=dyn_head, bone2Name=backpack_root, swing1LimitDegrees=12, swing2LimitDegrees=5, twistLimitDegrees=45, twistMotion=Locked}
]
